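# Animation Blueprint for the RankedHalt backpack cosmetic, animating floating pieces and shifting the pack as the wearer's spine bends.
# Each update builds real-time sine offsets on X and Z for the crown, trapezoid, metal piece and main bones.
# Spine_05's yaw relative to spine_01, measured against ref pose, maps to pack roll and Y shift.
# Chained Modify Bone nodes add these to fixed offsets; computed GravityOverride and RB Alpha are unused in the AnimGraph.
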AnimBlueprint /BRCosmetics/Accessories/FORT_Backpacks/Backpack_M_MED_RankedHalt/Meshes/M_MED_RankedHalt_Pack_AnimBP
target_skeleton=&/M_MED_RankedHalt_Pack_Skeleton
graphs [
 AnimGraph {nodes [
   n1 = "Transform (Modify) Bone - Bone: pack"(ComponentPose=Local To Component(LocalPose=Input Pose()), Translation="0,4,0", Rotation="0,0,0") {
    translationMode=BMM_Additive
    translationSpace=BCS_BoneSpace
    alphaBoolBlend {blendOption=Linear}
    boneToModify {boneName=pack}
    translation {y=4}
   }
   n2 = "Transform (Modify) Bone - Bone: pack"(
    ComponentPose=@n1
    Translation=Get Backpack Translation() -> (Backpack Translation:Vector)
    Rotation=Get Backpack Rotate() -> (Backpack Rotate:Rotator)
   ) {
    rotationMode=BMM_Additive
    rotationSpace=BCS_BoneSpace
    translationMode=BMM_Additive
    translationSpace=BCS_BoneSpace
    alphaBoolBlend {blendOption=Linear}
    boneToModify {boneName=pack}
   }
   n3 = "Transform (Modify) Bone - Bone: base_trapezoid"(ComponentPose=@n2, Alpha=1, Translation="0.5,0,0", Rotation="0,0,0", Scale="1,1,1") {
    rotationMode=BMM_Additive
    rotationSpace=BCS_BoneSpace
    translationMode=BMM_Additive
    translationSpace=BCS_BoneSpace
    alphaBoolBlend {blendOption=Linear}
    boneToModify {boneName=base_trapezoid}
    translation {x=0.5}
   }
   n4 = "Transform (Modify) Bone - Bone: dyn_crown"(ComponentPose=@n3, Alpha=1, Translation="0,3,0", Rotation="0,0,0", Scale="1,1,1") {
    translationMode=BMM_Additive
    translationSpace=BCS_BoneSpace
    alphaBoolBlend {blendOption=Linear}
    boneToModify {boneName=dyn_crown}
    translation {y=3}
   }
   n5 = "Transform (Modify) Bone - Bone: dyn_metal_piece"(ComponentPose=@n4, Alpha=1, Translation="0,3,0", Rotation="0,0,0", Scale="1,1,1") {
    translationMode=BMM_Additive
    translationSpace=BCS_BoneSpace
    alphaBoolBlend {blendOption=Linear}
    boneToModify {boneName=dyn_metal_piece}
    translation {y=3}
   }
   n6 = "Transform (Modify) Bone - Bone: dyn_crown"(ComponentPose=@n5, Alpha=1, Translation="0,0.9,0", Rotation="0,0,0", Scale="1,1,1") {
    translationMode=BMM_Additive
    translationSpace=BCS_BoneSpace
    alphaBoolBlend {blendOption=Linear}
    boneToModify {boneName=dyn_crown}
    translation {y=0.9}
   }
   n7 = "Transform (Modify) Bone - Bone: dyn_main"(ComponentPose=@n6, Alpha=1, Translation=Get Translate Main() -> (Translate Main:Vector), Rotation="0,0,0", Scale="1,1,1") {translationMode=BMM_Additive, translationSpace=BCS_BoneSpace, alphaBoolBlend {blendOption=Linear}, boneToModify {boneName=dyn_main}}
   n8 = "Transform (Modify) Bone - Bone: dyn_crown"(ComponentPose=@n7, Alpha=1, Translation=Get Translate Crown() -> (Translate Crown:Vector), Rotation="0,0,0", Scale="1,1,1") {translationMode=BMM_Additive, translationSpace=BCS_BoneSpace, alphaBoolBlend {blendOption=Linear}, boneToModify {boneName=dyn_crown}}
   n9 = "Transform (Modify) Bone - Bone: dyn_trapezoid"(ComponentPose=@n8, Alpha=1, Translation=Get Translate Trapezoid() -> (Translate Trapezoid:Vector), Rotation="0,0,0", Scale="1,1,1") {translationMode=BMM_Additive, translationSpace=BCS_BoneSpace, alphaBoolBlend {blendOption=Linear}, boneToModify {boneName=dyn_trapezoid}}
   n10 = "Transform (Modify) Bone - Bone: dyn_metal_piece"(ComponentPose=@n9, Alpha=1, Translation=Get Translate Metal Piece() -> (Translate Metal Piece:Vector), Rotation="0,0,0", Scale="1,1,1") {translationMode=BMM_Additive, translationSpace=BCS_BoneSpace, alphaBoolBlend {blendOption=Linear}, boneToModify {boneName=dyn_metal_piece}}
   Output Pose(Result=Component To Local(ComponentPose=@n10)) {layerGroup=None}
  ]}
 EventGraph {nodes [
   Event BlueprintUpdateAnimation()
   > n1 = Sequence!(execute) -> (then_0, then_1, then_2:Exec, then_3:Exec, then_4:Exec, then_5:Exec)
   n2 = GetRealTimeSeconds()
   n3 = "Sin (Radians)"(A=float * float(A=float * float(A=1, B=-1.2), B=@n2) {is=title K2Node_PromotableOperator})
   n4 = "Sin (Radians)"(A=float * float(A=float * float(A=0.4, B=-0.8), B=@n2))
   Set Translate Crown(execute=@n1.then_1, Translate Crown:Vector=MakeVector(X=float / float(A=@n3, B=-1.5), Y=0, Z=float / float(A=@n4, B=-1))) -> (Output_Get:Vector)
   n5 = Gravity Override State Function DEPRECATED(
    execute=@n1.then_0
    CharacterPartAnimInstance=Self-Reference()
    GameTimeInSeconds=GetGameTimeInSeconds()
    GravityOverrideParams=Get Gravity Override Params()
   )
   > Set GravityOverride(GravityOverride=@n5)
   > Set RB Alpha(RB Alpha:double=SelectFloat(A=0.1, B=0.5, bPickA=Get bIsSkyDiving())) -> (Output_Get:double)
   n6 = GetRealTimeSeconds()
   n7 = "Sin (Radians)"(A=float * float(A=float * float(A=1, B=-1.6), B=@n6))
   n8 = "Sin (Radians)"(A=float * float(A=float * float(A=0.4, B=-0.8), B=@n6))
   Set Translate Trapezoid(execute=@n1.then_2, Translate Trapezoid:Vector=MakeVector(X=float / float(A=@n7, B=2), Y=0, Z=float / float(A=@n8, B=1.4))) -> (Output_Get:Vector)
   n9 = GetRealTimeSeconds()
   n10 = "Sin (Radians)"(A=float * float(A=float * float(A=0.4, B=-0.8), B=@n9))
   n11 = "Sin (Radians)"(A=float * float(A=float * float(A=1, B=-1.2), B=@n9))
   Set Translate Metal Piece(execute=@n1.then_3, Translate Metal Piece:Vector=MakeVector(X=float / float(A=@n11, B=-0.9), Y=0, Z=float / float(A=@n10, B=-0.7))) -> (Output_Get:Vector)
   n12 = GetRealTimeSeconds()
   n13 = "Sin (Radians)"(A=float * float(A=float * float(A=1, B=-1.4), B=@n12))
   n14 = "Sin (Radians)"(A=float * float(A=float * float(A=0.8, B=-0.8), B=@n12))
   Set Translate Main(execute=@n1.then_4, Translate Main:Vector=MakeVector(X=float / float(A=@n13, B=4.3), Y=0, Z=float / float(A=@n14, B=6.3))) -> (Output_Get:Vector)
   n15 = GetOwnerSkeletalMeshComponent()
   n16 = BreakRotator(InRot=BreakTransform(InTransform=GetDeltaTransformFromRefPose(self=@n15, BoneName=spine_05, BaseName=spine_01)).Rotation)
   Set Backpack Rotate(execute=@n1.then_5, Backpack Rotate:Rotator=MakeRotator(Roll=MapRangeClamped(Value=@n16.Yaw, InRangeA=0, InRangeB=-40, OutRangeB=25))) -> (Output_Get:Rotator)
   > Set Backpack Translation(Backpack Translation:Vector=MakeVector(Y=MapRangeClamped(Value=@n16.Yaw, InRangeA=0, InRangeB=-30, OutRangeB=-5))) -> (Output_Get:Vector)
   TryGetPawnOwner()
   n17 = float * float(A=1, B=1.1)
   float / float(A="Sin (Radians)"(A=float * float(A=@n17, B=@n2)), B=20)
   n18 = float * float(A=1, B=1.3)
   float / float(A="Sin (Radians)"(A=float * float(A=@n18, B=@n12)), B=20)
   n19 = float * float(A=0.7, B=1.3)
   float / float(A="Sin (Radians)"(A=float * float(A=@n19, B=@n6)), B=4)
   n20 = float * float(A=1, B=2)
   float / float(A="Sin (Radians)"(A=float * float(A=@n20, B=@n9)), B=4)
  ]}
]
variables [
 Backpack Rotate {type=Rotator}
 Backpack Translation {type=Vector}
 Gravity Override Params {type=GravityOverrideParamsStruct}
 GravityOverride {type=Vector}
 RB Alpha {type=number}
 Translate Crown {type=Vector}
 Translate Main {type=Vector}
 Translate Metal Piece {type=Vector}
 Translate Trapezoid {type=Vector}
]
